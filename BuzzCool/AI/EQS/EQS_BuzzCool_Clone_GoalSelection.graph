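# EQS query that selects a goal for the BuzzCool Clone AI, preferring enemies over large donut objects.
# The first option gathers recently perceived, non-DBNO enemies and scores them on perception, pathfinding, facing and proximity.
# It also filters and penalizes frustrated goals and drops enemies outside a 2D leash around the clone's leash center.
# The untested fallback option finds tagged BP_BuzzCool_Donut_Large actors within a data-bound radius of the querier.
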EnvQuery /BuzzCool/AI/EQS/EQS_BuzzCool_Clone_GoalSelection
options [
 {
  generator {
   name=FortQueryGenerator_Enemies_0
   class=FortQueryGenerator_Enemies
   properties {
    bAddEnemiesFromAbilityRange=false
    bAutoSortTests=true
    bCanRunAsync=false
    bIgnoreDBNOPawns=true
    bIgnoreSleepingAIs=false
    bIncludeMarkedSleepingAIs=false
    bIncludePropagatedEnemies=false
    bPerceivedEnemiesOnly=true
    bSleepCapableAIUsePerceivedEnemiesOnly=true
    envQueryResultNormalizationOption=Default
    optionName=""
    abilityTags {gameplayTags []}
    distanceTags {gameplayTags []}
    maxTimeSincePerceived {
     dataBinding=&self:FortQueryGenerator_Enemies_0.FortAIDataProvider_ScalableFloat_0
     dataField=FloatValue
     defaultValue=15
    }
   }
  }
  tests [
   FortQueryTest_PerceptionAll_0 {
    class=FortQueryTest_PerceptionAll
    properties {
     clampMaxType=None
     clampMinType=None
     scoreClampMax=0
     scoringFactor=2
     additionalSenseScores {&Class'&AthenaAISense_Hearing'=1}
     maxSenseAge {
      dataBinding=&self:FortQueryTest_PerceptionAll_0.FortAIDataProvider_ScalableFloat_1
      dataField=FloatValue
      defaultValue=0
     }
     minSenseAge {
      dataBinding=&self:FortQueryTest_PerceptionAll_0.FortAIDataProvider_ScalableFloat_0
      dataField=FloatValue
      defaultValue=0
     }
     senseScores [0,0,0,0,0,0]
    }
   }
   EnvQueryTest_PathfindingBatch_1 {
    class=EnvQueryTest_PathfindingBatch
    properties {filterClass=/Game/Athena/AI/NavFilters/NavFilter_NPC_AllNavmesh.NavFilter_NPC_AllNavmesh_C, testPurpose=Score}
   }
   FortQueryTest_GoalFrustrationDiscouragement_0 {
    class=FortQueryTest_GoalFrustrationDiscouragement
    properties {
     bDefineReferenceValue=false
     bScoreEncounterGoals=true
     bScoreEnemies=true
     bScoreSpecificAssignments=false
     bScoreWorldGoals=true
     boolValue=true
     clampMaxType=SpecifiedValue
     clampMinType=SpecifiedValue
     filterType=Maximum
     floatValueMax=1
     floatValueMin=0
     multipleContextFilterOp=AllPass
     multipleContextScoreOp=AverageScore
     normalizationType=Absolute
     referenceValue=0
     scoreClampMax=1
     scoreClampMin=0
     scoringEquation=Constant
     scoringFactor=-1
     testComment="When a Goal can't be reached, we can choose to Frustrate on (IGNORE) that Goal for a period of time."
     testPurpose=FilterAndScore
     assignmentIds []
     goalActorTagQuery {autoDescription="", tokenStreamVersion=0, userDescription="", queryTokenStream [], tagDictionary []}
     prohibitedAssignmentIds []
    }
   }
   EnvQueryTest_Dot_5 {
    class=EnvQueryTest_Dot
    properties {
     clampMaxType=SpecifiedValue
     clampMinType=SpecifiedValue
     scoreClampMax=0
     scoreClampMin=0
     scoringFactor=1
     testComment=Prefer Goal in front
     lineA {dirMode=TwoPoints, lineFrom=&EnvQueryContext_Querier, lineTo=&EnvQueryContext_Item}
     lineB {dirMode=Rotation, lineTo=&FortQueryContext_Goal, rotation=&EnvQueryContext_Querier}
    }
   }
   FortQueryTest_GoalDistance_1 {
    class=FortQueryTest_GoalDistance
    properties {
     bScoreEncounterGoals=true
     bScoreEnemies=true
     bScoreSpecificAssignments=false
     bScoreWorldGoals=true
     clampMaxType=None
     clampMinType=None
     filterType=Range
     floatValueMax=0
     scoreClampMax=0
     scoreClampMin=0
     scoringEquation=InverseLinear
     scoringFactor=1
     testComment=Prefer closer targets.
     testPurpose=Score
     assignmentIds []
    }
   }
   FortQueryTest_GoalDistance_3 {
    class=FortQueryTest_GoalDistance
    properties {
     bScoreEncounterGoals=true
     bScoreEnemies=true
     bScoreSpecificAssignments=false
     bScoreWorldGoals=true
     clampMaxType=None
     clampMinType=None
     distanceTo=&/EnvQueryContext_BuzzCool_Clone_LeashCenter.EnvQueryContext_BuzzCool_Clone_LeashCenter_C
     filterType=Maximum
     scoreClampMax=0
     scoreClampMin=0
     scoringEquation=Linear
     scoringFactor=1
     testComment=Leash: ignore outside the leash.
     testMode=Distance2D
     testPurpose=Filter
     assignmentIds []
     floatValueMax {
      dataBinding=&self:FortQueryTest_GoalDistance_3.FortAIDataProvider_ScalableFloat_0
      dataField=FloatValue
      defaultValue=0
     }
    }
   }
  ]
 }
 {
  generator {
   name=FortQueryGenerator_GoalTrackableAIObjects_0
   class=FortQueryGenerator_GoalTrackableAIObjects
   properties {
    bAutoSortTests=true
    bCanRunAsync=false
    envQueryResultNormalizationOption=Default
    optionName=""
    searchCenter=&EnvQueryContext_Querier
    searchedActorClass=/BuzzCool/Gameplay/BP_BuzzCool_Donut_Large.BP_BuzzCool_Donut_Large_C
    assignmentIdentifier {assignmentTeam=Monster, assignmentType=World, assignmentGameplayTags {gameplayTags []}}
    requiredTag {tagName=Item.BuzzCool.SpawnActor}
    searchRadius {
     dataBinding=&self:FortQueryGenerator_GoalTrackableAIObjects_0.FortAIDataProvider_ScalableFloat_0
     dataField=FloatValue
     defaultValue=0
    }
   }
  }
  tests []
 }
]
