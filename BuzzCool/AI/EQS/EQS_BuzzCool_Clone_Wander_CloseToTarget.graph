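# EQS query that picks a wander point near the goal for BuzzCool clone AI.
# Generates a donut around the goal with data-bound radii and density, projected and path-tested using a water-avoiding nav filter.
# Filters out points too close to the querier or nearby AI pawns.
# Inverse-linear scoring favors points near both goal and querier, contradicting the "farthest from Self" comment, with a random test for variety.
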EnvQuery /BuzzCool/AI/EQS/EQS_BuzzCool_Clone_Wander_CloseToTarget
options [{
  generator {
   name=EnvQueryGenerator_Donut_0
   class=EnvQueryGenerator_Donut
   properties {
    bUseSpiralPattern=false
    center=&FortQueryContext_Goal
    innerRadius {
     dataBinding=&self:EnvQueryGenerator_Donut_0.FortAIDataProvider_ScalableFloat_1
     dataField=FloatValue
     defaultValue=300
    }
    numberOfRings {
     dataBinding=&self:EnvQueryGenerator_Donut_0.FortAIDataProvider_ScalableFloat_3
     dataField=IntValue
     defaultValue=3
    }
    outerRadius {
     dataBinding=&self:EnvQueryGenerator_Donut_0.FortAIDataProvider_ScalableFloat_2
     dataField=FloatValue
     defaultValue=1000
    }
    pointsPerRing {
     dataBinding=&self:EnvQueryGenerator_Donut_0.FortAIDataProvider_ScalableFloat_4
     dataField=IntValue
     defaultValue=8
    }
    projectionData {
     extentX=0
     navigationFilter=&NavFilter_NPC_AvoidAllWater_C
     postProjectionVerticalOffset=0
    }
   }
  }
  tests [
   EnvQueryTest_PathfindingBatch_3 {
    class=EnvQueryTest_PathfindingBatch
    properties {filterClass=&NavFilter_NPC_AvoidAllWater_C}
   }
   EnvQueryTest_Distance_6 {
    class=EnvQueryTest_Distance
    properties {
     distanceTo=&EnvQueryContext_Querier
     filterType=Minimum
     scoringEquation=InverseLinear
     scoringFactor=1
     testComment=Pick point farthest from Self
     testMode=Distance2D
     testPurpose=FilterAndScore
     floatValueMin {
      dataBinding=&self:EnvQueryTest_Distance_6.FortAIDataProvider_ScalableFloat_0
      dataField=FloatValue
      defaultValue=0
     }
    }
   }
   EnvQueryTest_Distance_0 {
    class=EnvQueryTest_Distance
    properties {
     distanceTo=&FortQueryContext_Goal
     filterType=Range
     floatValueMin=0
     scoringEquation=InverseLinear
     scoringFactor=1
     testComment=""
     testMode=Distance2D
     testPurpose=Score
    }
   }
   FortQueryTest_Random_2 {class=FortQueryTest_Random, properties {floatValueMax=0.1, scoringFactor=0.5}}
   EnvQueryTest_Distance_7 {
    class=EnvQueryTest_Distance
    properties {
     distanceTo=&FortQueryContext_NearbyAIPawns
     filterType=Minimum
     floatValueMin=300
     scoringEquation=InverseLinear
     scoringFactor=3
     testComment=Try not to overlap AI Pawns
     testMode=Distance2D
     testPurpose=Filter
    }
   }
  ]
 }]
