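# EQS query that picks a free-wander destination for a BuzzCool clone within its leash.
# Three fallback options generate donut points around the querier, projecting with ExcludeWater, AvoidAllWater, then AnyTraversable nav filters, though all path-test with ExcludeWater.
# Tests reject points outside the leash, unreachable, or near other AI pawns or their destinations.
# Scoring favors points far from self, near leash center and friendlies/players, avoiding U-turns, with random variation.
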EnvQuery /BuzzCool/AI/EQS/EQS_BuzzCool_Clone_Wander_Freely_Location
options [
 {
  generator {
   name=EnvQueryGenerator_Donut_2
   class=EnvQueryGenerator_Donut
   properties {
    bUseSpiralPattern=false
    center=&EnvQueryContext_Querier
    arcDirection {
     lineFrom=null
     lineTo=/Game/Athena/AI/EQS/Contexts/EnvQueryContext_Querier_RandomPointOnUnitCircle_2D.EnvQueryContext_Querier_RandomPointOnUnitCircle_2D_C
    }
    innerRadius {
     dataBinding=&self:EnvQueryGenerator_Donut_2.FortAIDataProvider_ScalableFloat_0
     dataField=FloatValue
     defaultValue=200
    }
    numberOfRings {
     dataBinding=&self:EnvQueryGenerator_Donut_2.FortAIDataProvider_ScalableFloat_2
     dataField=IntValue
     defaultValue=3
    }
    outerRadius {
     dataBinding=&self:EnvQueryGenerator_Donut_2.FortAIDataProvider_ScalableFloat_1
     dataField=FloatValue
     defaultValue=1500
    }
    pointsPerRing {
     dataBinding=&self:EnvQueryGenerator_Donut_2.FortAIDataProvider_ScalableFloat_3
     dataField=IntValue
     defaultValue=6
    }
    projectionData {
     extentX=100
     navigationFilter=/Game/Athena/AI/NavFilters/NavFilter_NPC_ExcludeWater.NavFilter_NPC_ExcludeWater_C
     postProjectionVerticalOffset=30
    }
   }
  }
  tests [
   EnvQueryTest_Distance_0 {
    class=EnvQueryTest_Distance
    properties {
     clampMinType=SpecifiedValue
     distanceTo=&/EnvQueryContext_BuzzCool_Clone_LeashCenter.EnvQueryContext_BuzzCool_Clone_LeashCenter_C
     filterType=Maximum
     floatValueMin=0
     scoreClampMax=2000
     scoreClampMin=2000
     scoringEquation=InverseLinear
     scoringFactor=0.75
     testComment=Inside leash
     testMode=Distance2D
     testPurpose=FilterAndScore
     floatValueMax {
      dataBinding=&self:EnvQueryTest_Distance_0.FortAIDataProvider_ScalableFloat_0
      dataField=FloatValue
      defaultValue=0
     }
    }
   }
   FortQueryTest_PathfindingBatch_0 {
    class=FortQueryTest_PathfindingBatch
    properties {
     bDefineReferenceValue=false
     boolValue=true
     clampMaxType=None
     clampMinType=None
     context=&EnvQueryContext_Querier
     filterClass=&NavFilter_NPC_ExcludeWater_C
     filterType=Match
     floatValueMax=1000
     floatValueMin=1000
     multipleContextFilterOp=AllPass
     multipleContextScoreOp=AverageScore
     navDataOverrideContext=null
     normalizationType=Absolute
     pathFromContext=true
     referenceValue=0
     scanRangeMultiplier=1.5
     scoreClampMax=0
     scoreClampMin=0
     scoringEquation=Constant
     scoringFactor=1
     skipUnreachable=true
     testComment=Throw out things you cannot path to.
     testMode=PathExist
     testPurpose=Filter
     navFilterTag {tagName=None}
    }
   }
   EnvQueryTest_Distance_1 {
    class=EnvQueryTest_Distance
    properties {
     distanceTo=&EnvQueryContext_Querier
     filterType=Range
     floatValueMin=0
     scoringEquation=Linear
     scoringFactor=1
     testComment=Pick point farthest from Self
     testMode=Distance2D
     testPurpose=Score
    }
   }
   EnvQueryTest_Dot_0 {
    class=EnvQueryTest_Dot
    properties {clampMaxType=SpecifiedValue, clampMinType=None, scoreClampMax=0.5, scoreClampMin=0, scoringFactor=1, testComment=Avoid 180 degrees turns}
   }
   FortQueryTest_Random_0 {class=FortQueryTest_Random, properties {floatValueMax=0.1, scoringFactor=1, testComment=Random Score}}
   EnvQueryTest_Distance_2 {
    class=EnvQueryTest_Distance
    properties {
     distanceTo=&FortQueryContext_NearbyAIPawns
     filterType=Minimum
     scoringEquation=InverseLinear
     scoringFactor=3
     testComment=Try not to overlap AI Pawns
     testMode=Distance3D
     testPurpose=Filter
     floatValueMin {
      dataBinding=&self:EnvQueryTest_Distance_2.FortAIDataProvider_ScalableFloat_0
      dataField=FloatValue
      defaultValue=300
     }
    }
   }
   EnvQueryTest_Distance_4 {
    class=EnvQueryTest_Distance
    properties {
     distanceTo=&FortQueryContext_NearbyAIPawnsMoveDestinations
     filterType=Minimum
     scoringEquation=Linear
     scoringFactor=1
     testComment=Not near other pawn move destinations
     testMode=Distance3D
     testPurpose=Filter
     floatValueMin {
      dataBinding=&self:EnvQueryTest_Distance_4.FortAIDataProvider_ScalableFloat_0
      dataField=FloatValue
      defaultValue=300
     }
    }
   }
   EnvQueryTest_Distance_5 {
    class=EnvQueryTest_Distance
    properties {
     distanceTo=/Script/FortniteAIServer.FortQueryContext_NearbyFriendlyAIPawnsAndPlayerPawns
     filterType=Range
     floatValueMin=0
     scoringEquation=InverseLinear
     scoringFactor=0.5
     testComment=""
     testMode=Distance2D
     testPurpose=Score
    }
   }
  ]
 }
 {
  generator {
   name=EnvQueryGenerator_Donut_0
   class=EnvQueryGenerator_Donut
   properties {
    bUseSpiralPattern=false
    center=&EnvQueryContext_Querier
    arcDirection {
     lineFrom=null
     lineTo=&EnvQueryContext_Querier_RandomPointOnUnitCircle_2D_C
    }
    innerRadius {
     dataBinding=&self:EnvQueryGenerator_Donut_0.FortAIDataProvider_ScalableFloat_0
     dataField=FloatValue
     defaultValue=200
    }
    numberOfRings {
     dataBinding=&self:EnvQueryGenerator_Donut_0.FortAIDataProvider_ScalableFloat_2
     dataField=IntValue
     defaultValue=3
    }
    outerRadius {
     dataBinding=&self:EnvQueryGenerator_Donut_0.FortAIDataProvider_ScalableFloat_1
     dataField=FloatValue
     defaultValue=1500
    }
    pointsPerRing {
     dataBinding=&self:EnvQueryGenerator_Donut_0.FortAIDataProvider_ScalableFloat_3
     dataField=IntValue
     defaultValue=6
    }
    projectionData {
     extentX=100
     navigationFilter=&NavFilter_NPC_AvoidAllWater_C
     postProjectionVerticalOffset=30
    }
   }
  }
  tests [
   EnvQueryTest_Distance_6 {
    class=EnvQueryTest_Distance
    properties {
     clampMinType=SpecifiedValue
     distanceTo=&EnvQueryContext_BuzzCool_Clone_LeashCenter_C
     filterType=Maximum
     floatValueMin=0
     scoreClampMax=2000
     scoreClampMin=2000
     scoringEquation=InverseLinear
     scoringFactor=0.75
     testComment=Inside leash
     testMode=Distance2D
     testPurpose=FilterAndScore
     floatValueMax {
      dataBinding=&self:EnvQueryTest_Distance_6.FortAIDataProvider_ScalableFloat_0
      dataField=FloatValue
      defaultValue=0
     }
    }
   }
   FortQueryTest_PathfindingBatch_1 {
    class=FortQueryTest_PathfindingBatch
    properties {
     bDefineReferenceValue=false
     boolValue=true
     clampMaxType=None
     clampMinType=None
     context=&EnvQueryContext_Querier
     filterClass=&NavFilter_NPC_ExcludeWater_C
     filterType=Match
     floatValueMax=1000
     floatValueMin=1000
     multipleContextFilterOp=AllPass
     multipleContextScoreOp=AverageScore
     navDataOverrideContext=null
     normalizationType=Absolute
     pathFromContext=true
     referenceValue=0
     scanRangeMultiplier=1.5
     scoreClampMax=0
     scoreClampMin=0
     scoringEquation=Constant
     scoringFactor=1
     skipUnreachable=true
     testComment=Throw out things you cannot path to.
     testMode=PathExist
     testPurpose=Filter
     navFilterTag {tagName=None}
    }
   }
   EnvQueryTest_Distance_7 {
    class=EnvQueryTest_Distance
    properties {
     distanceTo=&EnvQueryContext_Querier
     filterType=Range
     floatValueMin=0
     scoringEquation=Linear
     scoringFactor=1
     testComment=Pick point farthest from Self
     testMode=Distance2D
     testPurpose=Score
    }
   }
   EnvQueryTest_Dot_1 {
    class=EnvQueryTest_Dot
    properties {clampMaxType=SpecifiedValue, clampMinType=None, scoreClampMax=0.5, scoreClampMin=0, scoringFactor=1, testComment=Avoid 180 degrees turns}
   }
   FortQueryTest_Random_1 {class=FortQueryTest_Random, properties {floatValueMax=0.1, scoringFactor=0.7, testComment=Random Score}}
   EnvQueryTest_Distance_9 {
    class=EnvQueryTest_Distance
    properties {
     distanceTo=&FortQueryContext_NearbyAIPawns
     filterType=Minimum
     scoringEquation=InverseLinear
     scoringFactor=3
     testComment=Try not to overlap AI Pawns
     testMode=Distance3D
     testPurpose=Filter
     floatValueMin {
      dataBinding=&self:EnvQueryTest_Distance_9.FortAIDataProvider_ScalableFloat_0
      dataField=FloatValue
      defaultValue=300
     }
    }
   }
   EnvQueryTest_Distance_13 {
    class=EnvQueryTest_Distance
    properties {
     distanceTo=&FortQueryContext_NearbyAIPawnsMoveDestinations
     filterType=Minimum
     scoringEquation=Linear
     scoringFactor=1
     testComment=Not near other pawn move destinations
     testMode=Distance3D
     testPurpose=Filter
     floatValueMin {
      dataBinding=&self:EnvQueryTest_Distance_13.FortAIDataProvider_ScalableFloat_0
      dataField=FloatValue
      defaultValue=300
     }
    }
   }
   EnvQueryTest_Distance_14 {
    class=EnvQueryTest_Distance
    properties {
     distanceTo=&FortQueryContext_NearbyFriendlyAIPawnsAndPlayerPawns
     filterType=Range
     floatValueMin=0
     scoringEquation=InverseLinear
     scoringFactor=0.5
     testComment=""
     testMode=Distance2D
     testPurpose=Score
    }
   }
  ]
 }
 {
  generator {
   name=EnvQueryGenerator_Donut_1
   class=EnvQueryGenerator_Donut
   properties {
    bUseSpiralPattern=false
    center=&EnvQueryContext_Querier
    arcDirection {
     lineFrom=null
     lineTo=&EnvQueryContext_Querier_RandomPointOnUnitCircle_2D_C
    }
    innerRadius {
     dataBinding=&self:EnvQueryGenerator_Donut_1.FortAIDataProvider_ScalableFloat_0
     dataField=FloatValue
     defaultValue=200
    }
    numberOfRings {
     dataBinding=&self:EnvQueryGenerator_Donut_1.FortAIDataProvider_ScalableFloat_2
     dataField=IntValue
     defaultValue=3
    }
    outerRadius {
     dataBinding=&self:EnvQueryGenerator_Donut_1.FortAIDataProvider_ScalableFloat_1
     dataField=FloatValue
     defaultValue=1500
    }
    pointsPerRing {
     dataBinding=&self:EnvQueryGenerator_Donut_1.FortAIDataProvider_ScalableFloat_3
     dataField=IntValue
     defaultValue=6
    }
    projectionData {
     extentX=100
     navigationFilter=&NavFilter_NPC_AnyTraversable_C
     postProjectionVerticalOffset=30
    }
   }
  }
  tests [
   EnvQueryTest_Distance_15 {
    class=EnvQueryTest_Distance
    properties {
     clampMinType=SpecifiedValue
     distanceTo=&EnvQueryContext_BuzzCool_Clone_LeashCenter_C
     filterType=Maximum
     floatValueMin=0
     scoreClampMax=2000
     scoreClampMin=2000
     scoringEquation=InverseLinear
     scoringFactor=0.75
     testComment=Inside leash
     testMode=Distance2D
     testPurpose=FilterAndScore
     floatValueMax {
      dataBinding=&self:EnvQueryTest_Distance_15.FortAIDataProvider_ScalableFloat_0
      dataField=FloatValue
      defaultValue=0
     }
    }
   }
   FortQueryTest_PathfindingBatch_3 {
    class=FortQueryTest_PathfindingBatch
    properties {
     bDefineReferenceValue=false
     boolValue=true
     clampMaxType=None
     clampMinType=None
     context=&EnvQueryContext_Querier
     filterClass=&NavFilter_NPC_ExcludeWater_C
     filterType=Match
     floatValueMax=1000
     floatValueMin=1000
     multipleContextFilterOp=AllPass
     multipleContextScoreOp=AverageScore
     navDataOverrideContext=null
     normalizationType=Absolute
     pathFromContext=true
     referenceValue=0
     scanRangeMultiplier=1.5
     scoreClampMax=0
     scoreClampMin=0
     scoringEquation=Constant
     scoringFactor=1
     skipUnreachable=true
     testComment=Throw out things you cannot path to.
     testMode=PathExist
     testPurpose=Filter
     navFilterTag {tagName=None}
    }
   }
   EnvQueryTest_Distance_16 {
    class=EnvQueryTest_Distance
    properties {
     distanceTo=&EnvQueryContext_Querier
     filterType=Range
     floatValueMin=0
     scoringEquation=Linear
     scoringFactor=1
     testComment=Pick point farthest from Self
     testMode=Distance2D
     testPurpose=Score
    }
   }
   EnvQueryTest_Dot_3 {
    class=EnvQueryTest_Dot
    properties {clampMaxType=SpecifiedValue, clampMinType=None, scoreClampMax=0.5, scoreClampMin=0, scoringFactor=1, testComment=Avoid 180 degrees turns}
   }
   FortQueryTest_Random_3 {class=FortQueryTest_Random, properties {floatValueMax=0.1, scoringFactor=0.7, testComment=Random Score}}
   EnvQueryTest_Distance_17 {
    class=EnvQueryTest_Distance
    properties {
     distanceTo=&FortQueryContext_NearbyAIPawns
     filterType=Minimum
     scoringEquation=InverseLinear
     scoringFactor=3
     testComment=Try not to overlap AI Pawns
     testMode=Distance3D
     testPurpose=Filter
     floatValueMin {
      dataBinding=&self:EnvQueryTest_Distance_17.FortAIDataProvider_ScalableFloat_0
      dataField=FloatValue
      defaultValue=300
     }
    }
   }
   EnvQueryTest_Distance_23 {
    class=EnvQueryTest_Distance
    properties {
     distanceTo=&FortQueryContext_NearbyAIPawnsMoveDestinations
     filterType=Minimum
     scoringEquation=Linear
     scoringFactor=1
     testComment=Not near other pawn move destinations
     testMode=Distance3D
     testPurpose=Filter
     floatValueMin {
      dataBinding=&self:EnvQueryTest_Distance_23.FortAIDataProvider_ScalableFloat_0
      dataField=FloatValue
      defaultValue=300
     }
    }
   }
   EnvQueryTest_Distance_24 {
    class=EnvQueryTest_Distance
    properties {
     distanceTo=&FortQueryContext_NearbyFriendlyAIPawnsAndPlayerPawns
     filterType=Range
     floatValueMin=0
     scoringEquation=InverseLinear
     scoringFactor=0.5
     testComment=""
     testMode=Distance2D
     testPurpose=Score
    }
   }
  ]
 }
]
